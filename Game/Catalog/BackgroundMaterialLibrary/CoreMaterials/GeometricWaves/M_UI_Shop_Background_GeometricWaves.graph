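# UI translucent material that draws the Geometric Waves background for Item Shop tiles, with colors supplied by the shared item shop background-params function.
# It builds a smoothstepped vertical color gradient, a parameterized radial gradient mask, a grunge texture broken up by blurred noise, and a scalable wave-lines texture.
# Named reroutes carry these layers, UV variants and a floor-stepped time value, with nested lerps compositing color and a rounded-corners function handling tile shape.
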Material /Game/Catalog/BackgroundMaterialLibrary/CoreMaterials/GeometricWaves/M_UI_Shop_Background_GeometricWaves
graph {
 properties {bAutomaticallySetUsageInEditor=false, blendMode=BLEND_Translucent, materialDomain=MD_UI, usedShadingModels=Default Lit}
 nodes [
  Constant3Vector() {constant=<rgba 0.491 0.4653 0.0763 1>}
  Constant3Vector() {constant=<rgba 0.5833 0.4841 0.1268 1>}
  Constant3Vector() {constant=<rgba 0.0395 0.0529 0.0252 1>}
  n1 = NamedRerouteUsage()
  LinearInterpolate(
   A=LinearInterpolate(A=NamedRerouteUsage(), B=NamedRerouteUsage(), Alpha=Multiply(A=@n1) {constB=0.2})
   B=NamedRerouteUsage()
   Alpha=NamedRerouteUsage()
  )
  MaterialFunctionCall!(Input Alpha=Constant() {r=1}, bOverrideCornerRadius, Corner Radius Override) {materialFunction=/FortUILibrary/Systems/ItemShopTile/Materials/MaterialFunctions/MF_UI_ItemShop_RoundedCorners}
  n2 = NamedRerouteUsage()
  n3 = TextureObject() {
   samplerType=SAMPLERTYPE_Alpha
   texture=/Game/Catalog/BackgroundMaterialLibrary/CoreMaterials/CrumpleFold/Textures/T_UI_DarkVortex_Background_Grunge
  }
  n4 = TextureSample(
   UVs=MaterialFunctionCall(UVs=@n2, Texture Scale=Constant() {r=0.3}) {materialFunction=&ScaleUVsByCenter}.UVs
   Tex=@n3
  ) {
   samplerType=SAMPLERTYPE_Alpha
   texture=&T_UI_DarkVortex_Background_Grunge
  }
  n5 = TextureSample(UVs=@n2, Tex=TextureObject() {samplerType=SAMPLERTYPE_Grayscale, texture=&LowResBlurredNoise}) {samplerType=SAMPLERTYPE_Grayscale}
  n6 = Multiply(A=@n5.R, B=@n5.R)
  n7 = Multiply(A=Multiply(A=@n4.R, B=LinearInterpolate(Alpha=@n6) {constA=0.1})) {constB=2}
  n8 = NamedRerouteUsage()
  NamedRerouteDeclaration(None=Saturate(None=LinearInterpolate(A=@n7, Alpha=@n8) {constB=0})) {name=Grunge Texture, nodeColor=<rgba 0 0 0 1>}
  n9 = MaterialFunctionCall() {materialFunction=/FortUILibrary/Systems/ItemShopTile/Materials/MaterialFunctions/MF_UI_ItemShop_BackgroundParams}
  NamedRerouteDeclaration(None=@n9.Color 1) {name=Primary Color, nodeColor=<rgba 0 0 0 1>}
  n10 = NamedRerouteUsage()
  n11 = Saturate(None=MaterialFunctionCall(Input=ComponentMask(None=@n10) {g=true}, Input Low=Constant(), Input High=Constant() {r=0.4}) {materialFunction=&RemapValue_0-1}.Result)
  NamedRerouteDeclaration(None=LinearInterpolate(A=NamedRerouteUsage(), B=NamedRerouteUsage(), Alpha=SmoothStep(Value=@n11))) {name=Background Gradient, nodeColor=<rgba 0 0 0 1>}
  n12 = MaterialFunctionCall(UV=NamedRerouteDeclaration(None=MaterialFunctionCall() {materialFunction=&GetUserInterfaceUV}.Normalized UV) {name=Normalized UV, nodeColor=<rgba 0 0 0 1>}) {materialFunction=&MF_UI_UVScaling}
  NamedRerouteDeclaration(None=@n12.Scaled to Fit) {name=Scale to Fit, nodeColor=<rgba 0 0 0 1>}
  NamedRerouteDeclaration(None=@n9.Color 2) {name=Secondary Color, nodeColor=<rgba 0 0 0 1>}
  n13 = MaterialFunctionCall!(ScreenUV, TextureSize, ScalePivotPoint) {materialFunction=/FortUILibrary/MaterialFunctions/Basics/MF_UI_ScreenSpaceUVScaling}
  NamedRerouteDeclaration(None=@n13.ScaleY) {name="Scale Y UV  Screen", nodeColor=<rgba 0 0 0 1>}
  NamedRerouteDeclaration(None=@n13.Scaled to Fill) {name="Scale to Fill  Screen", nodeColor=<rgba 0 0 0 1>}
  NamedRerouteDeclaration(None=@n13.Scaled to Fit) {name=Scale to Fit Screen, nodeColor=<rgba 0 0 0 1>}
  NamedRerouteDeclaration(None=@n12.ScaleY) {name=Scale Y UV, nodeColor=<rgba 0 0 0 1>}
  n14 = NamedRerouteUsage()
  n15 = ScalarParameter() {defaultValue=2, group=Radial Gradient, parameterName=Radial Gradient Scale}
  n16 = ScalarParameter() {group=Radial Gradient, parameterName=Radial Gradient Offset X}
  n17 = ScalarParameter() {defaultValue=-0.1, group=Radial Gradient, parameterName=Radial Gradient Offset Y}
  n18 = MaterialFunctionCall(UV=Add(
    A=MaterialFunctionCall(UVs=@n14, Texture Scale=@n15) {materialFunction=&ScaleUVsByCenter}.UVs
    B=AppendVector(A=@n16, B=@n17)
   ) {constB=0.5}) {materialFunction=&MF_UI_UVScaling}
  NamedRerouteDeclaration(None=Saturate(None=MaterialFunctionCall(UV=@n18.Scaled To Fill) {materialFunction=&MF_UI_RadialGradient}.Smooth)) {name=Radial Gradient Mask, nodeColor=<rgba 0 0 0 1>}
  NamedRerouteDeclaration(None=@n12.ScaleX) {name=Scale X UV, nodeColor=<rgba 0 0 0 1>}
  NamedRerouteDeclaration(None=@n12.Scaled To Fill) {name=Scale to Fill, nodeColor=<rgba 0 0 0 1>}
  n19 = MaterialFunctionCall(UVs=NamedRerouteUsage(), Texture Scale=ScalarParameter() {defaultValue=0.2, group=Lines Texture, parameterName=Texture Size}) {materialFunction=&ScaleUVsByCenter}
  n20 = AppendVector(A=Add(B=ComponentMask(None=@n19.UVs) {r=true}), B=Add(A=ComponentMask(None=@n19.UVs) {g=true}) {constB=0.1})
  NamedRerouteDeclaration(None=Saturate(None=TextureSample(UVs=@n20) {samplerType=SAMPLERTYPE_Alpha, texture=/Game/Catalog/BackgroundMaterialLibrary/Textures/T_UI_GeometricWaves}.RGB)) {name=Lines, nodeColor=<rgba 0 0 0 1>}
  NamedRerouteDeclaration(None=@n9.Color 3) {name=Accent Color, nodeColor=<rgba 0 0 0 1>}
  n21 = Constant() {r=0.3}
  n22 = Floor(None=Multiply(A=Time(), B=@n21) {constB=0.2})
  NamedRerouteDeclaration(None=OneMinus(None=Divide(A=@n22, B=@n21))) {name=Stepped time, nodeColor=<rgba 0 0 0 1>}
 ]
}
